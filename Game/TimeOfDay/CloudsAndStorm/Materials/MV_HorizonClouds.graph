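# Additive volumetric cloud material that forms a ring of horizon clouds from a torus signed distance field.
# World position is rotated, scaled and height-offset, then warped by curl noise before entering the torus function.
# Two scales of biased, power-curved Voronoi noise erode the distance field, which is remapped into density.
# An optional radial depth gradient alters color, and the advanced volumetric output exposes dual-phase, per-sample multi-scattering lighting.
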Material /Game/TimeOfDay/CloudsAndStorm/Materials/MV_HorizonClouds
graph {
 properties {
  bUsedWithVolumetricCloud=true
  blendMode=BLEND_Additive
  materialDomain=MD_Volume
  pixelDepthOffsetMode=PDOM_Legacy
  usedShadingModels=Default Lit
 }
 nodes [
  n1 = TextureCoordinate()
  n2 = Constant()
  Add(A=TextureSample(UVs=AppendVector(A=@n1, B=@n2)) {samplerType=SAMPLERTYPE_LinearColor, texture=/Game/TimeOfDay/CloudsAndStorm/Textures/VT_PerlinWorley}.R) {constB=0}
  CloudSampleAttribute()
  Constant2Vector()
  n3 = WorldPosition()
  n4 = MaterialFunctionCall(UVs=@n3.XY, Rotation Center=Constant2Vector(), "Rotation Angle (0-1)"=ScalarParameter() {parameterName=Rotation_0_to_1}) {materialFunction=&CustomRotator}
  n5 = Constant()
  n6 = ScalarParameter() {defaultValue=-0.8, parameterName=Z}
  n7 = Subtract(
   A=Divide(A=AppendVector(A=@n4.Rotated Values, B=@n3.Z), B=ScalarParameter() {defaultValue=1, parameterName=GlobalScale})
   B=MaterialFunctionCall(X=@n5, Y=@n5, Z=MaterialFunctionCall!(Km=@n6) {materialFunction=/Engine/Functions/Engine_MaterialFunctions01/Units/Km-to-Cm}.Cm) {materialFunction=&MakeFloat3}.Result
  )
  Divide(A=@n7) {constB=1000000}
  n8 = ScalarParameter() {defaultValue=20, parameterName=A}
  Divide(A=@n8) {constB=-2}
  n9 = VectorParameter() {parameterName=Col, defaultValue=<rgba 1 1 1 1>}
  n10 = WorldPosition()
  n11 = Length(None=ComponentMask(None=MaterialFunctionCall!(Cm=@n10.XYZ) {materialFunction=/Game/TimeOfDay/CloudsAndStorm/Materials/MaterialFunctions/Cm_to_Km_Vector}.Km) {g=true, r=true}) {is=class MaterialExpressionLength}
  n12 = ScalarParameter() {defaultValue=5, parameterName=B}
  n13 = Subtract(A=@n8, B=@n12)
  n14 = Add(A=@n8, B=Multiply(A=@n12, B=ScalarParameter() {defaultValue=0.333, parameterName=BackDepthMult}))
  n15 = OneMinus(None=Saturate(None=MaterialFunctionCall(Input=@n11, Input Low=@n13, Input High=@n14) {materialFunction=&RemapValueRange}.Result))
  n16 = ScalarParameter() {defaultValue=2, parameterName=DepthPower}
  LinearInterpolate(
   A=ComponentMask(None=@n9.RGB) {b=true, g=true, r=true}
   B=MaterialFunctionCall(X=@n9.R, Y=Power(Base=@n15, Exp=@n16), Z=Constant()) {materialFunction=&MakeFloat3}.Result
   Alpha=ScalarParameter() {parameterName=EnableDepth}
  )
  MaterialFunctionCall!(Cm=@n7) {materialFunction=/Engine/Functions/Engine_MaterialFunctions01/Units/Cm-to-Km}
  n17 = ScalarParameter() {defaultValue=64, parameterName=NoiseScale}
  MaterialFunctionCall(Km=@n17) {materialFunction=&Km-to-Cm}
  n18 = ScalarParameter() {defaultValue=8, parameterName=NoiseScale2}
  MaterialFunctionCall(Km=@n18) {materialFunction=&Km-to-Cm}
  MaterialFunctionCall!(Cm=@n7) {materialFunction=/Game/TimeOfDay/CloudsAndStorm/Materials/MaterialFunctions/Cm-to-Km}
  MaterialFunctionCall!(Km) {materialFunction=/Game/TimeOfDay/CloudsAndStorm/Materials/MaterialFunctions/Km-to-Cm}
  n19 = MaterialFunctionCall(Cm=@n7) {materialFunction=&Cm_to_Km_Vector}
  n20 = ScalarParameter() {defaultValue=0.02, parameterName=CurlScalke}
  n21 = Subtract(A=TextureSample(UVs=Multiply(A=@n19.Km, B=@n20)) {samplerType=SAMPLERTYPE_LinearColor, texture=/Volumetrics/Content/VolumeTextures/Textures/VT_CurlNoise}.RGB) {constB=0.5}
  n22 = ScalarParameter() {parameterName=CurlGain}
  n23 = Subtract(A=@n19.Km, B=Multiply(A=Multiply(A=@n21, B=@n22), B=Constant3Vector() {constant=<rgba 1 1 0.2 0>}))
  n24 = MaterialFunctionCall!(
   ZNoiseScale
   TorusVerticalOffset
   torusCentre=Constant2Vector()
   ta=@n8
   tb=@n12
   HeightZPos=ComponentMask(None=@n23) {b=true}
   worldXY_UV=ComponentMask(None=@n23) {g=true, r=true}
   VectorNoiseOffset
  ) {materialFunction=/Game/TimeOfDay/CloudsAndStorm/Materials/MaterialFunctions/MF_Torus_SDF1}
  n25 = OneMinus(None=TextureSample(UVs=Divide(A=@n19.Km, B=@n18)) {samplerType=SAMPLERTYPE_LinearColor, texture=/Game/TimeOfDay/CloudsAndStorm/Textures/FN_VT_Voronoi}.R)
  n26 = ScalarParameter() {
   defaultValue=2
   desc="Inverse Power Curve on the erosion noise to make rounder. Higher Values = more puffy"
   parameterName=Detail Noise Puffiness
  }
  n27 = ScalarParameter() {defaultValue=0.2259, desc=Change Centre point of noise. 0.5 means equal erosion/expansion, parameterName=Detail Noise Bias}
  n28 = Subtract(A=OneMinus(None=Power(Base=@n25, Exp=@n26)), B=@n27) {constB=0.5}
  n29 = Divide(A=Constant() {r=1}, B=OneMinus(None=@n27))
  n30 = Multiply(A=OneMinus(None=Multiply(A=@n28, B=@n29)), B=Multiply(A=ScalarParameter() {defaultValue=0.78, parameterName=Erode2}) {constB=-1})
  n31 = TextureSample(UVs=Divide(A=@n19.Km, B=@n17) {constB=1000000}) {samplerType=SAMPLERTYPE_LinearColor, texture=&FN_VT_Voronoi}
  n32 = ScalarParameter() {defaultValue=1, parameterName=ErodePowerCurve}
  n33 = Power(Base=OneMinus(None=@n31.R), Exp=@n32)
  n34 = ScalarParameter() {defaultValue=0.25, parameterName=NoiseBias}
  n35 = Constant() {r=1}
  n36 = OneMinus(None=@n34)
  n37 = Multiply(
   A=Multiply(A=Subtract(A=@n33, B=@n34) {constB=0.5}, B=Divide(A=@n35, B=@n36))
   B=Multiply(A=ScalarParameter() {defaultValue=4.7, parameterName=Erode}) {constB=-1}
  )
  n38 = Add(A=@n30, B=@n37)
  Multiply(
   A=MaterialFunctionCall(
    Input=Subtract(A=@n24.TorusSDF, B=@n38)
    Input Low=ScalarParameter() {defaultValue=-0.5, parameterName=Min}
    Input High=ScalarParameter() {parameterName=Max}
    Target Low=Constant() {r=1}
    Target High=Constant()
   ) {materialFunction=&RemapValueRange}.Result
   B=ScalarParameter() {defaultValue=0.032, parameterName=density}
  )
  n39 = Subtract(A=@n31.R, B=@n34) {constB=0.5}
  Multiply(A=Power(Base=Abs(None=@n39), Exp=@n32), B=Sign(None=@n39))
  OneMinus(None=@n33)
  ScalarParameter() {defaultValue=0.5, parameterName=AO}
  Subtract(A=@n37, B=@n30)
  VolumetricAdvancedMaterialOutput!(
   PhaseG=ScalarParameter() {defaultValue=0.3, parameterName=Phase1}
   PhaseG2=ScalarParameter() {defaultValue=-0.3, parameterName=Phase2}
   PhaseBlend
   MultiScatteringContribution=ScalarParameter() {defaultValue=1, parameterName=Scatter}
   MultiScatteringOcclusion=ScalarParameter() {defaultValue=0.2, parameterName=Occlusion}
   MultiScatteringEccentricity=ScalarParameter() {defaultValue=1, parameterName=Eccentricity}
   ConservativeDensity
  ) {
   bClampMultiScatteringContribution=false
   constMultiScatteringContribution=1
   constPhaseBlend=0.5
   multiScatteringApproximationOctaveCount=1
   perSamplePhaseEvaluation=true
  }
 ]
}
