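# Unlit material in the time-of-day cloud and storm tech art that reduces a cloud color capture to one averaged color, apparently for writing a line of a LUT.
# It samples the RT_CloudColorCaptureForLUT render target at nine fixed UVs in a 3x3 grid around the center.
# The RGB samples are summed and divided by nine, so the result is uniform and independent of the mesh or screen UVs.
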Material /Game/TimeOfDay/CloudsAndStorm/TechArt/M_CaptureSingleLUTLIne
graph {
 properties {bAutomaticallySetUsageInEditor=false, pixelDepthOffsetMode=PDOM_Legacy, shadingModel=MSM_Unlit, usedShadingModels=Unlit}
 nodes [
  n1 = Constant2Vector() {g=0.35, r=0.35}
  n2 = Constant2Vector() {g=0.35, r=0.5}
  n3 = Add(
   A=Add(
    A=TextureSample(UVs=@n1) {samplerType=SAMPLERTYPE_LinearColor, texture=&/RT_CloudColorCaptureForLUT}.RGB
    B=TextureSample(UVs=@n2) {samplerType=SAMPLERTYPE_LinearColor, texture=&RT_CloudColorCaptureForLUT}.RGB
   )
   B=TextureSample(UVs=Constant2Vector() {g=0.35, r=0.65}) {samplerType=SAMPLERTYPE_LinearColor, texture=&RT_CloudColorCaptureForLUT}.RGB
  )
  n4 = Constant2Vector() {g=0.5, r=0.35}
  n5 = Constant2Vector() {g=0.5, r=0.5}
  n6 = Add(
   A=Add(
    A=TextureSample(UVs=@n4) {samplerType=SAMPLERTYPE_LinearColor, texture=&RT_CloudColorCaptureForLUT}.RGB
    B=TextureSample(UVs=@n5) {samplerType=SAMPLERTYPE_LinearColor, texture=&RT_CloudColorCaptureForLUT}.RGB
   )
   B=TextureSample(UVs=Constant2Vector() {g=0.5, r=0.65}) {samplerType=SAMPLERTYPE_LinearColor, texture=&RT_CloudColorCaptureForLUT}.RGB
  )
  n7 = Add(
   A=TextureSample(UVs=Constant2Vector() {g=0.65, r=0.35}) {samplerType=SAMPLERTYPE_LinearColor, texture=&RT_CloudColorCaptureForLUT}.RGB
   B=TextureSample(UVs=Constant2Vector() {g=0.65, r=0.5}) {samplerType=SAMPLERTYPE_LinearColor, texture=&RT_CloudColorCaptureForLUT}.RGB
  )
  n8 = TextureSample(UVs=Constant2Vector() {g=0.65, r=0.65}) {samplerType=SAMPLERTYPE_LinearColor, texture=&RT_CloudColorCaptureForLUT}
  Divide(A=Add(A=Add(A=@n3, B=@n6), B=Add(A=@n7, B=@n8.RGB))) {constB=9}
 ]
}
