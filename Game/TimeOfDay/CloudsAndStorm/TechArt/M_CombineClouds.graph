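# Unlit alpha-composite material that combines four render targets, apparently cloud layers going by its name.
# It packs the R, G and B channels of three targets into one color, one channel from each.
# The fourth target's red channel, rescaled and saturated, forms a mask that blends white toward that color and also feeds divide and invert nodes.
# No links to material outputs appear, and a debug value display and stray constant remain.
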Material /Game/TimeOfDay/CloudsAndStorm/TechArt/M_CombineClouds
graph {
 properties {blendMode=BLEND_AlphaComposite, pixelDepthOffsetMode=PDOM_Legacy, shadingModel=MSM_Unlit, usedShadingModels=Unlit}
 nodes [
  Constant2Vector() {g=0.99, r=0.8}
  n1 = MaterialFunctionCall(
   X=TextureSample() {samplerType=SAMPLERTYPE_LinearColor, texture=&/RT_2d}.R
   Y=TextureSample() {samplerType=SAMPLERTYPE_LinearColor, texture=&/RT_2d1}.G
   Z=TextureSample() {samplerType=SAMPLERTYPE_LinearColor, texture=&/RT_2d2}.B
  ) {materialFunction=&MakeFloat3}
  n2 = TextureSample() {samplerType=SAMPLERTYPE_LinearColor, texture=&/RT_2d3}
  n3 = Saturate(None=Divide(A=@n2.R) {constB=0.49})
  Divide(A=@n1.Result, B=@n3)
  LinearInterpolate(A=Constant3Vector() {constant=<rgba 1 1 1 0>}, B=@n1.Result, Alpha=@n3)
  MaterialFunctionCall(Number=@n2.R) {materialFunction=&DebugScalarValues}
  OneMinus(None=@n3)
 ]
}
