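# M_CombineDepth is an unlit tech-art material in the clouds and storm folder that packs a render target value with its gradient direction.
# It samples the green channel of RT_2d3, takes its screen-space X and Y derivatives, and normalizes them into a two-component direction.
# That direction is appended to the sampled value as a three-component vector.
# No link to a material output is shown, and a two-component constant sits unconnected.
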Material /Game/TimeOfDay/CloudsAndStorm/TechArt/M_CombineDepth
graph {
 properties {pixelDepthOffsetMode=PDOM_Legacy, shadingModel=MSM_Unlit, usedShadingModels=Unlit}
 nodes [
  n1 = TextureSample() {samplerType=SAMPLERTYPE_LinearColor, texture=&/RT_2d3}
  n2 = DDX(Value=@n1.G)
  n3 = DDY(Value=@n1.G)
  AppendVector(
   A=@n1.G
   B=Normalize(VectorInput=MaterialFunctionCall(X=@n2, Y=@n3) {materialFunction=&MakeFloat2}.Result)
  )
  Constant2Vector() {g=0.99, r=0.8}
 ]
}
