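# Material that unwraps the CRT_Clouds cube texture into a 2D long-lat strip, apparently a tech-art utility for the time-of-day clouds.
# UVs are squeezed vertically by the Yffset parameter into a band ending at mid-height, then shifted horizontally by Rot.
# UVToLongLat turns the result into the cube lookup direction.
# Clamp, Saturate and VerticalOffset branches feed nothing, and no material output connection appears.
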Material /Game/TimeOfDay/CloudsAndStorm/TechArt/M_CubeTo2D
graph {
 properties {pixelDepthOffsetMode=PDOM_Legacy, usedShadingModels=Default Lit}
 nodes [
  n1 = TextureCoordinate()
  n2 = Constant2Vector() {g=0.5}
  n3 = Constant() {r=1}
  n4 = ScalarParameter() {defaultValue=8, parameterName=Yffset}
  n5 = Add(
   A=Divide(
    A=Subtract(A=@n1, B=@n2) {constB=0.5}
    B=MaterialFunctionCall(X=@n3, Y=@n4) {materialFunction=&MakeFloat2}.Result
   ) {constB=1.5}
   B=Constant2Vector() {g=0.5}
  ) {constB=0.5}
  n6 = MaterialFunctionCall(X=Constant(), Y=Divide(A=Constant() {r=0.5}, B=@n4)) {materialFunction=&MakeFloat2}
  n7 = ScalarParameter() {defaultValue=0.25, parameterName=Rot}
  n8 = Constant()
  n9 = MaterialFunctionCall!(UV=Add(
    A=Subtract(A=@n5, B=@n6.Result) {constB=0.5}
    B=MaterialFunctionCall(X=@n7, Y=@n8) {materialFunction=&MakeFloat2}.Result
   )) {materialFunction=/Engine/Functions/Engine_MaterialFunctions01/Coordinates/UVToLongLat}
  Clamp(None=@n9.Result, Min=Constant2Vector() {g=0})
  Constant2Vector() {g=10, r=1}
  Constant2Vector() {g=0.1}
  MaterialFunctionCall(X=Constant(), Y=ScalarParameter() {parameterName=VerticalOffset}) {materialFunction=&MakeFloat2}
  n10 = ScalarParameter() {defaultValue=0.25, parameterName=Rot}
  n11 = ScalarParameter() {defaultValue=4, parameterName=add}
  Saturate(None=Add(B=MaterialFunctionCall(X=@n10, Y=@n11) {materialFunction=&MakeFloat2}.Result))
  TextureSample(UVs=@n9.Result) {texture=&/CRT_Clouds}
 ]
}
