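# Editor utility actor that bakes volumetric horizon clouds into textures for the Fortnite time-of-day system.
# Its construction script drives a cloud material instance from noise, erosion, offset and rotation variables.
# Bake cube-captures clouds lit at 6, 12 and 18 hours plus a depth-enabled pass, flattens them to 2D, then saves combined cloud and depth textures.
# A separate captureLUT function sweeps 24 hours, drawing captured columns into a LUT render target.
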EditorUtilityBlueprint /Game/TimeOfDay/CloudsAndStorm/TechArt/EUA_BP_HorizonCloudBaker
graphs [
 1 EnableEditing {nodes [
   n1 = Self-Reference()
   Set Actor Hidden In Game!(execute, self:Actor Object Reference=@n1, bNewHidden:bool=true) -> (then) {is=title K2Node_CallFunction}
   > Set Actor Hidden In Game(self=@n1, bNewHidden=false)
   1 EnableEditing() {is=title K2Node_FunctionEntry}
   > SetVisibility!(
    execute
    self:Scene Component Object Reference=Get SkyAtmosphere() -> (SkyAtmosphere:Sky Atmosphere Component Object Reference)
    bNewVisibility:bool=false
    bPropagateToChildren:bool=false
   ) -> (then)
   > SetVisibility(
    self=Get VolumetricCloud() -> (VolumetricCloud:Volumetric Cloud Component Object Reference)
    bNewVisibility=false
    bPropagateToChildren=false
   )
   > SetVisibility(self=Get VolumetricCloud(), bNewVisibility=true, bPropagateToChildren=false)
   n2 = Get TextureTarget!(self:Scene Capture Component Cube Object Reference=Get SceneCaptureComponentCube() -> (SceneCaptureComponentCube:Scene Capture Component Cube Object Reference)) -> (TextureTarget:Texture Render Target Cube Object Reference)
   Set SizeX!(execute, SizeX:int=64, self:Texture Render Target Cube Object Reference=@n2) -> (then, Output_Get:int=0)
   > Set SizeX(SizeX=1024, self=@n2) -> (Output_Get=0)
  ]}
 "2 Bake!" {nodes [
   ExecuteConsoleCommand(Command=ke * SetCloudsCustomTime 1)
   n1 = Get As Fort Day Sequence Actor() -> (As Fort Day Sequence Actor:Fort Day Sequence Actor Object Reference)
   > SetTimeOfDayPreview!(execute, self:Day Sequence Actor Object Reference=@n1, InHours:float=6) -> (then)
   > UpdateComponentsAndMaterialParameters(self:Fort Day Sequence Actor Object Reference=@n1)
   > CaptureScene(self:Scene Capture Component 2D Object Reference=Get SceneCaptureComponent2D() -> (SceneCaptureComponent2D:Scene Capture Component 2D Object Reference))
   > n2 = BeginDrawCanvasToRenderTarget!(
    execute
    WorldContextObject:Object Reference
    TextureRenderTarget:Texture Render Target 2D Object Reference=&/RT_LUT
   ) -> (then, Canvas:Canvas Object Reference, Size:Vector 2D Structure, Context:Draw To Render Target Context Structure)
   > Draw Material!(
    execute
    self:Canvas Object Reference=@n2
    RenderMaterial:Material Interface Object Reference=&/M_CaptureSingleLUTLIne
    ScreenPosition:Vector 2D Structure
    ScreenSize:Vector 2D Structure=(X=0,Y=0.1)
    CoordinatePosition:Vector 2D Structure
    CoordinateSize:Vector 2D Structure=(X=1,Y=1)
    Rotation:float=0
    PivotPoint:Vector 2D Structure=(X=0.5,Y=0.5)
   ) -> (then)
   n3 = 1 EnableEditing() {is=title K2Node_CallFunction}
   GetEditorWorld(execute=[@n3.then, @n34.then], self=Get UnrealEditorSubsystem())
   > n4 = Sequence!(execute) -> (
    then_0
    then_1
    then_2:Exec
    then_3:Exec
    then_4:Exec
    then_5:Exec
    then_6:Exec
    then_7:Exec
    then_8:Exec
    then_9:Exec
    then_10:Exec
    then_11:Exec
    then_12:Exec
   )
   > 1 EnableEditing()
   > n5 = Cast To FortDaySequenceActor!(
    execute
    Object:Fort Time Of Day Manager Interface Interface=GetGlobalTimeOfDayManager!(WorldContextObject:Object Reference) -> (ReturnValue:Fort Time Of Day Manager Interface Interface)
   ) -> (then, CastFailed, AsFort Day Sequence Actor:Fort Day Sequence Actor Object Reference) {is=title K2Node_DynamicCast}
   > n6 = Set As Fort Day Sequence Actor(As Fort Day Sequence Actor:Fort Day Sequence Actor Object Reference=@n5) -> (Output_Get:Fort Day Sequence Actor Object Reference)
   > Set TOD_Orig(TOD_Orig:float=GetTimeOfDayPreview!(self:Day Sequence Actor Object Reference=@n6) -> (ReturnValue:float)) -> (Output_Get:float)
   > Set SunComponent(SunComponent:Directional Light Component Object Reference=Get SunComponent!(self:Fort Day Sequence Actor Object Reference=Get As Fort Day Sequence Actor()) -> (SunComponent:Directional Light Component Object Reference)) -> (Output_Get:Directional Light Component Object Reference)
   > n7 = GetConsoleVariableIntValue!(execute, VariableName:String=MakeLiteralString(Value=r.VolumetricRenderTarget.Mode)) -> (then, ReturnValue:int=0)
   > n8 = Set Cloud Mode(Cloud Mode:int=@n7) -> (Output_Get:int=0)
   > PrintString(InString=Append(A="Cloud Mode is: ", B="To String (Integer)"(InInt=@n8)))
   > "Set Console Variable By Name (Int)"!(execute, InCommandName:String=MakeLiteralString(Value=r.VolumetricRenderTarget.Mode), InValue:int=3) -> (then, ReturnValue:bool) {is=title K2Node_CallFunction}
   > n9 = Safe Array Get!(
    " ":Exec
    Array:Array of Sky Light Component Object References=Get Components By Class!(
     self:Actor Object Reference=Get Day Manager() -> (Day Manager:Actor Object Reference)
     ComponentClass:Actor Component Class Reference=/Script/Engine.SkyLightComponent
    ) -> (ReturnValue:Array of Sky Light Component Object References)
    Index:int
   ) -> ("  ":Exec, Result:Sky Light Component Object Reference) {is=title K2Node_MacroInstance}
   > n10 = Set TimeOfDaySkylight(TimeOfDaySkylight:Sky Light Component Object Reference=@n9) -> (Output_Get:Sky Light Component Object Reference)
   > SetHiddenInGame!(execute, self:Scene Component Object Reference=[@n9, @n10], NewHidden:bool=true, bPropagateToChildren:bool=false) -> (then)
   Reroute Node(InputPin=@n4.then_1)
   n11 = GetEditorWorld(execute=@n4.then_11, self=Get UnrealEditorSubsystem())
   > ClearRenderTarget2D!(
    execute
    WorldContextObject:Object Reference
    TextureRenderTarget:Texture Render Target 2D Object Reference=Get CombinedRenderTarget() -> (CombinedRenderTarget:Texture Render Target 2D Object Reference)
    ClearColor:Linear Color Structure=(R=0,G=0,B=0,A=1)
   ) -> (then)
   > DrawMaterialToRenderTarget!(
    execute
    WorldContextObject:Object Reference=@n11
    TextureRenderTarget:Texture Render Target 2D Object Reference=Get CombinedRenderTarget()
    Material:Material Interface Object Reference=&/M_CombineDepth
   ) -> (then)
   > SaveAsset!(execute, self, TextureName:String=MakeLiteralString(Value=T_HorizonDepthMap)) -> (then)
   n12 = Get As Fort Day Sequence Actor()
   SetTimeOfDayPreview(execute=@n4.then_6, self=@n12, InHours=18)
   > UpdateComponentsAndMaterialParameters(self=@n12)
   n13 = Get World Rotation!(self:Scene Component Object Reference=Get SunComponent!() -> (SunComponent)) -> (ReturnValue:Rotator) {is=title K2Node_CallFunction}
   n14 = Get DirectionalLightPM() -> (DirectionalLightPM:Directional Light Component Object Reference)
   > Set World Rotation!(execute, self:Scene Component Object Reference=@n14, NewRotation:Rotator=@n13, bSweep:bool=false, bTeleport:bool=false) -> (then, SweepHitResult:Hit Result Structure) {is=title K2Node_CallFunction}
   > PrintString(InString="To String (Rotator)"!(InRot:Rotator=@n13) -> (ReturnValue:String))
   n15 = Get TextureTarget(self=Get SceneCaptureComponentCube())
   > Set SizeX(SizeX=64, self=@n15) -> (Output_Get=0)
   > Set SizeX(SizeX=1024, self=@n15) -> (Output_Get=0)
   > Enable Light and Capture!(
    execute
    self
    LightToCapture:Scene Component Object Reference=@n14
    RT_2D:Texture Render Target 2D Object Reference=&/RT_2d2
    OnorOff:bool=true
    EnableCustomSkylight:bool=false
   ) -> (then)
   Branch(execute=@n4.then_12, Condition=false)
   n16 = Get As Fort Day Sequence Actor()
   n17 = BreakTopLevelAssetPath!(TopLevelAssetPath:"Top Level Asset Path Structure (by ref)"="Get Class Path Name (Top Level Asset Path)"!(Class:Object Class Reference=Get Class(Object=@n16) -> (ReturnValue:Fort Day Sequence Actor Class Reference)) -> (ReturnValue:Top Level Asset Path Structure) {is=title K2Node_CallFunction}) -> (PathString:String)
   > PrintString(InString=Append(A="############################", B=@n17))
   > n18 = GetAssetsByPath!(
    execute:Exec
    self:Object Reference=@n16
    PackagePath:Name=String To Name(InString=@n17)
    bRecursive:bool=true
    bIncludeOnlyOnDiskAssets:bool=true
   ) -> (then:Exec, OutAssetData:Array of Asset Data Structures, ReturnValue:bool) {is=title K2Node_Message}
   > n19 = For Each Loop(Array:Array of Asset Data Structures=@n18) -> (Array Element:Asset Data Structure)
   n20 = GetAsset(InAssetData=@n19)
   > PrintString(InString=Append(A="##########################", B=GetDisplayName(Object=@n20)))
   n21 = Get MID() -> (MID:Material Instance Dynamic Object Reference)
   SetScalarParameterValue(execute=@n4.then_7, self=@n21, ParameterName=EnableDepth, Value=1)
   > Enable Light and Capture(LightToCapture=Get SkyAtmosphere(), RT_2D=&/RT_2d3, OnorOff=false, EnableCustomSkylight=true)
   > SetScalarParameterValue(self=@n21, ParameterName=EnableDepth, Value=0)
   SetTimeOfDayManagerVisibility!(execute=@n4.then_8, self:Fort Time Of Day Manager Interface Interface=Get As Fort Day Sequence Actor(), bVisible:bool=true) -> (then)
   > SetVisibility(self=Get SkyAtmosphere(), bNewVisibility=false, bPropagateToChildren=true)
   > SetVisibility(self=Get VolumetricCloud(), bNewVisibility=false, bPropagateToChildren=true)
   > SetTimeOfDayPreview(self=Get As Fort Day Sequence Actor(), InHours=Get TOD_Orig() -> (TOD_Orig:float))
   > "Set Console Variable By Name (Int)"(InCommandName=MakeLiteralString(Value=r.VolumetricRenderTarget.Mode), InValue=Get Cloud Mode() -> (Cloud Mode:int=0))
   > ExecuteConsoleCommand(Command=ke * SetCloudsCustomTime 0)
   > n22 = Safe Array Get(Array=Get Components By Class(self=Get Day Manager(), ComponentClass=&SkyLightComponent))
   > n23 = Set TimeOfDaySkylight(TimeOfDaySkylight=@n22)
   > SetHiddenInGame(self=[@n22, @n23], NewHidden=false, bPropagateToChildren=false)
   n24 = GetEditorWorld(execute=@n4.then_9, self=Get UnrealEditorSubsystem())
   > ClearRenderTarget2D(TextureRenderTarget=Get CombinedRenderTarget(), ClearColor=(R=0,G=0,B=0,A=1))
   > DrawMaterialToRenderTarget(WorldContextObject=@n24, TextureRenderTarget=Get CombinedRenderTarget(), Material=&/M_CombineClouds)
   > SaveAsset(TextureName=MakeLiteralString(Value=T_CloudCombined))
   n25 = Get As Fort Day Sequence Actor()
   SetTimeOfDayPreview(execute=@n4.then_4, self=@n25, InHours=6)
   > UpdateComponentsAndMaterialParameters(self=@n25)
   n26 = Get World Rotation(self=Get SunComponent())
   > Set World Rotation(
    self=Get DirectionalLightAM() -> (DirectionalLightAM:Directional Light Component Object Reference)
    NewRotation=@n26
    bSweep=false
    bTeleport=false
   )
   > PrintString(InString="To String (Vector)"!(InVec:Vector=Get LightDirection!(self:Fort Day Sequence Actor Object Reference=@n25) -> (LightDirection:Vector)) -> (ReturnValue:String))
   n27 = Get TextureTarget(self=Get SceneCaptureComponentCube())
   > Set SizeX(SizeX=64, self=@n27) -> (Output_Get=0)
   > Set SizeX(SizeX=1024, self=@n27) -> (Output_Get=0)
   > Enable Light and Capture(LightToCapture=Get DirectionalLightAM(), RT_2D=&/RT_2d, OnorOff=true, EnableCustomSkylight=false)
   n28 = Get As Fort Day Sequence Actor()
   SetTimeOfDayPreview(execute=@n4.then_5, self=@n28, InHours=12)
   > UpdateComponentsAndMaterialParameters(self=@n28)
   n29 = Get World Rotation(self=Get SunComponent())
   n30 = Get DirectionalLightNoon() -> (DirectionalLightNoon:Directional Light Component Object Reference)
   > Set World Rotation(self=@n30, NewRotation=@n29, bSweep=false, bTeleport=false)
   > PrintString(InString="To String (Vector)"(InVec=Get LightDirection(self=@n28)))
   n31 = Get TextureTarget(self=Get SceneCaptureComponentCube())
   > Set SizeX(SizeX=64, self=@n31) -> (Output_Get=0)
   > Set SizeX(SizeX=1024, self=@n31) -> (Output_Get=0)
   > Enable Light and Capture(LightToCapture=@n30, RT_2D=&/RT_2d1, OnorOff=true, EnableCustomSkylight=false)
   n32 = Get TextureTarget(self=Get SceneCaptureComponentCube())
   Set SizeX(execute=@n4.then_2, SizeX=64, self=@n32) -> (Output_Get=0)
   > Set SizeX(SizeX=1024, self=@n32) -> (Output_Get=0)
   SetVisibility(execute=@n4.then_3, self=Get SkyAtmosphere(), bNewVisibility=true, bPropagateToChildren=false)
   n33 = Get As Fort Day Sequence Actor()
   > PrintString(InString=GetDisplayName(Object=@n33))
   > SetTimeOfDayManagerVisibility(self=@n33, bVisible=false)
   n34 = "2 Bake!"() {is=title K2Node_FunctionEntry}
   "To String (Rotator)"(InRot=@n29)
   "To String (Rotator)"(InRot=@n29)
   "To String (Rotator)"(InRot=@n26)
   Create Event()
   Reroute Node(InputPin:Wildcard) -> (OutputPin:Wildcard)
   Reroute Node()
   Get LightDirection(self=@n12)
  ]}
 3 GoToBakedTexture {nodes [3 GoToBakedTexture() {is=title K2Node_FunctionEntry}]}
 Enable Light and Capture {nodes [
   n1 = Enable Light and Capture!() -> (
    then
    LightToCapture:Scene Component Object Reference
    RT_2D:Texture Render Target 2D Object Reference
    OnorOff:bool=true
    EnableCustomSkylight:bool=~
   ) {is=title K2Node_FunctionEntry}
   > SetVisibility(self=@n1, bNewVisibility=@n1.OnorOff, bPropagateToChildren=false)
   n2 = Self-Reference()
   > Set Actor Hidden In Game(self=@n2, bNewHidden=true)
   > Set Actor Hidden In Game(self=@n2, bNewHidden=false)
   n3 = Self-Reference()
   n4 = Get Day Manager()
   > Set Actor Hidden In Game(self=[@n3, @n4], bNewHidden=false)
   > Set Actor Hidden In Game(self=[@n3, @n4], bNewHidden=true)
   > n5 = Branch(Condition=@n1.EnableCustomSkylight)
   > SetHiddenInGame(self=Get TimeOfDaySkylight() -> (TimeOfDaySkylight:Sky Light Component Object Reference), NewHidden=true, bPropagateToChildren=false)
   > SetHiddenInGame(self=Get SkyLight() -> (SkyLight:Sky Light Component Object Reference), NewHidden=true, bPropagateToChildren=false)
   > n6 = SetHiddenInGame(self=Get SkyLight(), NewHidden=false, bPropagateToChildren=false)
   n7 = Get SceneCaptureComponentCube()
   CaptureScene(execute=[@n6.then, @n5.else], self:Scene Capture Component Cube Object Reference=@n7)
   > n8 = GetEditorWorld(self=Get UnrealEditorSubsystem())
   > ClearRenderTarget2D(WorldContextObject=@n8, TextureRenderTarget=@n1.RT_2D, ClearColor=(R=0,G=0,B=0,A=1))
   > DrawMaterialToRenderTarget(WorldContextObject=@n8, TextureRenderTarget=@n1.RT_2D, Material=&/M_CubeTo2D)
   > SetVisibility(self=@n1, bNewVisibility=NOT Boolean(A=@n1.OnorOff), bPropagateToChildren=false)
   > Branch(Condition=@n1.EnableCustomSkylight)
   > SetHiddenInGame(self=Get SkyLight(), NewHidden=true, bPropagateToChildren=false)
   > SetHiddenInGame(self=Get SkyLight(), NewHidden=false, bPropagateToChildren=false)
   > SetHiddenInGame(self=Get TimeOfDaySkylight(), NewHidden=false, bPropagateToChildren=false)
   GetEditorWorld(self=Get UnrealEditorSubsystem())
   Get TextureTarget(self=@n7)
  ]}
 EventGraph {nodes [
   Event BeginPlay()
   Event ActorBeginOverlap()
   Event Tick!() -> (OutputDelegate, then, DeltaSeconds:float)
  ]}
 GetOrCreateTexture2D {nodes [
   n1 = GetOrCreateTexture2D!() -> (then, PackagePath:String, Asset Name:String) {is=title K2Node_FunctionEntry}
   > n2 = Sequence!(execute) -> (then_0, then_1)
   > PrintString(InString=@n1)
   n3 = Append!(A=@n1, B=/, C:String=@n1.Asset Name, D:String) -> (ReturnValue)
   > n4 = DoesAssetExist(AssetPath=@n3)
   > n5 = Branch(Condition=@n4)
   > n6 = LoadAsset(AssetPath=@n3)
   > n7 = Cast To Texture2D(Object=@n6)
   > Branch(Condition=IsValid(Object=@n6))
   > n8 = Set Texture_R(Texture_R:Texture 2D Object Reference=@n7) -> (Output_Get:Texture 2D Object Reference)
   > n9 = DeleteLoadedAsset!(execute, AssetToDelete:Object Reference=@n8) -> (then, ReturnValue:bool)
   n10 = GetAssetTools(execute=[@n9.then, @n5.else])
   > n11 = CreateAsset(self=@n10, AssetName=@n1.Asset Name, PackagePath=@n1, AssetClass=/Script/Engine.Texture2D)
   > n12 = Cast To Texture2D(Object=@n11)
   > PrintString(InString=GetDisplayName(Object=@n11))
   > Set Texture_R(Texture_R=@n12)
   Return Node!(execute=@n2.then_1, Texture:Texture 2D Object Reference=Get Texture_R() -> (Texture_R:Texture 2D Object Reference))
   Get AssetEditorSubsystem()
   Get AssetEditorUISubsystem() -> (ReturnValue:Asset Editor UISubsystem Object Reference) {is=title K2Node_GetEditorSubsystem}
   Get UnrealEditorSubsystem()
  ]}
 SaveAsset {nodes [
   n1 = SaveAsset!() -> (then, TextureName:String) {is=title K2Node_FunctionEntry}
   > Branch(Condition="Get SaveAsset (Disablef or testing)"() -> ("SaveAsset (Disablef or testing)":bool))
   > n2 = StripRightMostFolderFromPath_Util!(
    execute
    Path:String=Get Object Path String(Object=&self)
    WorldContextObject:Object Reference
    Strip Count:int=1
    __WorldContext:Object Reference
   ) -> (then, FinalPath1:String, AssetName:String)
   > n3 = StripRightMostFolderFromPath_Util(Path=@n2, Strip Count=1)
   n4 = Append!(A=@n3, B=/Textures) -> (ReturnValue)
   > n5 = GetOrCreateTexture2D_Util!(execute, PackagePath:String=@n4, Asset Name:String=@n1, WorldContextObject:Object Reference, __WorldContext:Object Reference) -> (then, Texture:Texture 2D Object Reference)
   > ConvertRenderTargetToTexture2DEditorOnly(WorldContextObject=Self-Reference(), RenderTarget=Get CombinedRenderTarget(), Texture=@n5)
   n6 = Get Object Path String(Object=@n5)
   > n7 = SaveAsset!(execute, AssetToSave:String=@n6, bOnlyIfIsDirty:bool=false) -> (then, ReturnValue:bool) {is=title K2Node_CallFunction}
   > Branch(Condition=@n7)
   > Set CompressionSettings(CompressionSettings=TC_BC7, self=@n5)
   > Set SRGB(SRGB=false, self=@n5)
   > SaveAsset(AssetToSave=@n6, bOnlyIfIsDirty=false)
   > Set Saved Asset(Saved Asset:Texture 2D Object Reference=@n5) -> (Output_Get:Texture 2D Object Reference)
   SyncBrowserToObjects()
   Append!(A=@n4, B=/, C:String=@n1) -> (ReturnValue)
  ]}
 UserConstructionScript {nodes [
   n1 = Cast To FortDaySequenceActor(Object=GetGlobalTimeOfDayManager())
   > UpdateComponentsAndMaterialParameters(self=@n1)
   Construction Script()
   > n2 = Branch(Condition=IsValid(Object=Get MID()))
   n3 = CreateDynamicMaterialInstance(
    execute=@n2.else
    Parent=Get main clouds material instance() -> (main clouds material instance:Material Instance Object Reference)
    CreationFlags=None
   )
   > n4 = Set MID(MID:Material Instance Dynamic Object Reference=@n3) -> (Output_Get:Material Instance Dynamic Object Reference)
   SetMaterial!(
    execute=[@n2.then, @n4.then]
    self:Volumetric Cloud Component Object Reference=Get VolumetricCloud()
    NewValue:Material Interface Object Reference=@n4
   ) -> (then)
   > SetScalarParameterValue(self=[Get MID(), @n4], ParameterName=NoiseScale, Value=Get Macro Noise Scale() -> (Macro Noise Scale:double))
   > SetScalarParameterValue(self=Get MID(), ParameterName=NoiseScale2, Value=Get Detail Noise Scale() -> (Detail Noise Scale:double))
   > SetScalarParameterValue(self=Get MID(), ParameterName=Erode, Value=Get Macro Erosion Gain() -> (Macro Erosion Gain:double))
   > SetScalarParameterValue(self=Get MID(), ParameterName=Erode2, Value=Get Detail Erosion Gain() -> (Detail Erosion Gain:double))
   n5 = Get MID()
   > SetScalarParameterValue(self=@n5, ParameterName=Z, Value="Get Vertical Offset (Kms)"() -> ("Vertical Offset (Kms)":double))
   > SetScalarParameterValue(self=@n5, ParameterName=Rotation_0_to_1, Value=Get WorldRotation() -> (WorldRotation:double))
   > 1 EnableEditing()
   PrintString(InString=Hello)
   Get Material!(self:Volumetric Cloud Component Object Reference=Get VolumetricCloud()) -> (Material:Material Interface Soft Object Reference)
   Get SkyAtmosphereComponent!(self:Fort Day Sequence Actor Object Reference=@n1) -> (SkyAtmosphereComponent:Sky Atmosphere Component Object Reference)
   Get main clouds material instance()
   Get distanceMaterial() -> (distanceMaterial:Material Object Reference)
   Set main clouds material instance(main clouds material instance:Material Instance Object Reference=@n3) -> (Output_Get:Material Instance Object Reference)
  ]}
 captureLUT {nodes [
   n1 = For Loop(execute=@n9.then, LastIndex=2047)
   n2 = MakeVector2D(X="To Float (Integer)"(InInt=@n1), Y=0)
   n3 = GetEditorWorld(execute=@n11.then, self=Get UnrealEditorSubsystem())
   > n4 = BeginDrawCanvasToRenderTarget(WorldContextObject=@n3, TextureRenderTarget=&RT_LUT)
   n5 = Cast To FortDaySequenceActor(execute=@n1.LoopBody, Object=GetGlobalTimeOfDayManager())
   n6 = Get SkyAtmosphereComponent(self=@n5)
   n7 = Draw Box!(
    execute
    self:Canvas Object Reference=@n4
    ScreenPosition:Vector 2D Structure=@n2
    ScreenSize:Vector 2D Structure=(X=1,Y=1)
    Thickness:float=1
    RenderColor:Linear Color Structure=LinearColor / LinearColor(
     A:Linear Color Structure=Get SkyLuminanceFactor!(self:Sky Atmosphere Component Object Reference=@n6) -> (SkyLuminanceFactor:Linear Color Structure)
     B:Linear Color Structure=(R=5,G=5,B=5,A=1)
    ) -> (ReturnValue:Linear Color Structure) {is=title K2Node_PromotableOperator}
   ) -> (then)
   EndDrawCanvasToRenderTarget!(
    execute=[@n7.then, @n14.then]
    WorldContextObject:Object Reference=@n3
    Context:"Draw To Render Target Context Structure (by ref)"=@n4.Context
   ) -> (then)
   captureLUT() {is=title K2Node_FunctionEntry}
   > n8 = Cast To FortDaySequenceActor(Object=GetGlobalTimeOfDayManager())
   > n9 = Set origtod(origtod:float=GetTimeOfDayPreview(self=@n8)) -> (Output_Get:float)
   n10 = Cast To FortDaySequenceActor(execute=@n1.Completed, Object=GetGlobalTimeOfDayManager())
   > SetTimeOfDayPreview(self=@n10, InHours=Get origtod() -> (origtod:float))
   int / int(A=@n1, B=31)
   Get As Fort Day Sequence Actor()
   n11 = CaptureScene(execute=@n13.then, self:Scene Capture Component 2D Object Reference=Get SceneCaptureComponent2D())
   n12 = float + float(A="To Float (Integer)"(InInt=@n1), B=0.5)
   SetTimeOfDayPreview(execute=@n5.then, self=@n5, InHours=float * float(A=float / float(A=@n12, B=2047), B=24) {is=title K2Node_PromotableOperator})
   > n13 = UpdateComponentsAndMaterialParameters(self=@n5)
   n14 = Draw Material(
    execute=@n4.then
    self=@n4
    RenderMaterial=&M_CaptureSingleLUTLIne
    ScreenPosition=@n2
    ScreenSize=(X=1,Y=256)
    CoordinatePosition=(X=0,Y=0)
    CoordinateSize=(X=1,Y=1)
    Rotation=0
    PivotPoint=(X=0.5,Y=0.5)
   )
  ]}
]
variables [
 AMRot {type=Rotator, default=<rot -26.96 180 180>}
 As Fort Day Sequence Actor {type=FortDaySequenceActor, default=null}
 Cloud Mode {type=integer}
 CombinedRenderTarget {type=TextureRenderTarget2D, default=&/RT_2d_Combine}
 Day Manager {type=Actor, default=null}
 Detail Erosion Gain {type=number, default=0.7}
 Detail Noise Scale {type=number, default=5.4}
 "Distance(Km)" {type=number}
 distanceMaterial {type=Material, default=/Game/TimeOfDay/CloudsAndStorm/Materials/MV_HorizonClouds, class=~, graph=~}
 Macro Erosion Gain {type=number, default=4.25}
 Macro Noise Scale {type=number, default=64}
 main clouds material instance {type=MaterialInstance, default=/Game/TimeOfDay/CloudsAndStorm/Materials/MV_HorizonClouds_Inst}
 MID {type=MaterialInstanceDynamic}
 NewVar {type=Transform, default {location=<0 0 0>, rotation=<rot 0 0 0>, scale=<1 1 1>}}
 NoonRot {type=Rotator, default=<rot -54.9908 -102.6765 -87.8072>}
 PMRot {type=Rotator, default=<rot -28.0021 0 0>}
 "SaveAsset (Disablef or testing)" {type=boolean, default=true}
 Saved Asset {type=Texture2D}
 SunComponent {type=DirectionalLightComponent, default=null}
 TimeOfDaySkylight {type=SkyLightComponent, default=null}
 TOD_Orig {type=number}
 "Vertical Offset (Kms)" {type=number}
 WorldRotation {type=number}
]
