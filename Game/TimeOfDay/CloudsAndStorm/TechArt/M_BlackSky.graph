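# M_BlackSky is a Default Lit material in the time-of-day clouds and storm tech art folder, named for a black sky.
# Its graph holds a single Constant node left at its default value, with no connection to a material output shown.
# Automatic usage setting in the editor is off and pixel depth offset uses legacy mode.
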Material /Game/TimeOfDay/CloudsAndStorm/TechArt/M_BlackSky
graph {properties {bAutomaticallySetUsageInEditor=false, pixelDepthOffsetMode=PDOM_Legacy, usedShadingModels=Default Lit}, nodes [Constant()]}
